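# Translucent unlit water material for mesh particles, apparently the wash effect when the VinderTech water umbrella opens.
# Panning foam normals drive cubemap reflection, sun/moon specular and Fresnel coloring, with height-based foam on top.
# Particle alpha erodes a gradient-masked foam and swirl opacity; mobile skips reflection and depth fade.
# Dynamic parameters control panning, tiling and a pivot-relative vertex offset weighted along U, before the TechArt accessories override function.
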Material /BRCosmetics/Gadgets/Assets/VinderTech_Umbrella/Umbrella_Water/M_Umbrella_Open_Wash
graph {
 properties {
  bUseMaterialAttributes=true
  bUsedWithMeshParticles=true
  bUsedWithNiagaraMeshParticles=true
  blendMode=BLEND_Translucent
  pixelDepthOffsetMode=PDOM_Legacy
  shadingModel=MSM_Unlit
  twoSided=true
  usedShadingModels=Unlit
 }
 nodes [
  n1 = TextureCoordinate()
  n2 = ScalarParameter() {defaultValue=1, parameterName=FoamTiling}
  n3 = VectorParameter() {parameterName=NormalTiling, defaultValue=<rgba 1 1 0 1>}
  n4 = AppendVector(A=@n3.R, B=@n3.R)
  n5 = DynamicParameter() {paramNames [WPO,PanAdd,Tile,Param4]}
  n6 = Multiply(A=Time(), B=ScalarParameter() {defaultValue=-1, parameterName=TimeScale})
  n7 = Multiply(A=ScalarParameter() {defaultValue=-0.2, parameterName=NormalX Scale}, B=@n6)
  n8 = ScalarParameter() {defaultValue=0.7, parameterName=NormalY Scale}
  n9 = Add(A=Multiply(A=Multiply(A=@n1, B=@n2), B=Multiply(A=@n4, B=@n5.Tile)), B=AppendVector(A=Add(A=@n5.PanAdd, B=@n7), B=Multiply(A=@n8, B=@n6)))
  n10 = TextureSample!(UVs=@n9, Tex, Bias=Constant() {r=2}, Apply View MipBias) {mipValueMode=TMVM_MipBias, samplerType=SAMPLERTYPE_Normal, texture=/Game/Athena/Apollo/Water/Textures/02_Rounded/02_WaterFoam_Veroni_base_N}
  n11 = Transform(None=MaterialFunctionCall(Normal=@n10.RGB, Flatness=ScalarParameter() {defaultValue=0.5, parameterName=Flattne}) {materialFunction=&FlattenNormal}.Result)
  n12 = TextureSampleParameterCube!(UVs=ReflectionVectorWS(CustomWorldNormal=@n11) {bNormalizeCustomWorldNormal=true}, Apply View MipBias) {
   parameterName=MaterialExpressionTextureSampleParameterCube_0
   samplerType=SAMPLERTYPE_LinearColor
   texture=/Game/Packages/Fortress_Sky/TexturesHDR/T_AthenaSkylight
  }
  n13 = VectorParameter() {parameterName=Cube Map Multiplier, defaultValue=<rgba 0.175 0.4871 1 0>}
  n14 = Multiply(
   A=Multiply(A=Fresnel() {baseReflectFraction=0.15, exponent=2}, B=Multiply(A=@n12.RGB, B=@n13.RGB) {constB=1.5})
   B=ComponentMask(None=MaterialFunctionCall() {materialFunction=/Engine/Functions/GetPostProcessFunctions/GetAmbientCubemapTint}.AmbientCubemapTint) {b=true, g=true, r=true}
  )
  n15 = Constant() {r=0.5}
  n16 = CollectionParameter() {
   collection=&FortniteMaterialParameters
   parameterName=SunAndMoonModelDirectionalVector
  }
  n17 = Clamp(None=DotProduct(A=ComponentMask(None=@n16) {b=true, g=true, r=true}, B=ReflectionVectorWS(CustomWorldNormal=@n11)))
  n18 = ScalarParameter() {defaultValue=60, parameterName=Spec Highlight Power}
  n19 = Add(
   A=ShadingPathSwitch(Default=Desaturation(None=@n14, Fraction=@n15), Mobile=Constant()) {desc=Migrated from FeatureLevelSwitch}
   B=Multiply(A=Power(Base=@n17, Exp=@n18) {constExponent=20}, B=ScalarParameter() {defaultValue=4, parameterName=Spec Intensity})
  )
  n20 = VectorParameter() {parameterName=Edge Color, defaultValue=<rgba 0 0.1825 0.4 1>}
  n21 = VectorParameter() {parameterName=Core Color, defaultValue=<rgba 0.0464 0.4948 0.3588 1>}
  n22 = ScalarParameter() {defaultValue=2, parameterName=exp}
  n23 = Saturate(None=Multiply(A=Fresnel(ExponentIn=@n22, Normal=@n11) {baseReflectFraction=0, exponent=0.25}))
  n24 = TextureSample!(UVs=@n9, Tex, Apply View MipBias) {samplerType=SAMPLERTYPE_LinearColor, texture=/Game/Athena/Apollo/Water/Textures/02_Rounded/02_WaterFoam_Veroni_base_H}
  n25 = ScalarParameter() {defaultValue=3.5, parameterName=FoamAmount}
  n26 = LinearInterpolate(
   A=Multiply(A=LinearInterpolate(A=@n20.RGB, B=@n21.RGB, Alpha=@n23), B=ScalarParameter() {defaultValue=0.7, parameterName=Brightness})
   B=VectorParameter() {parameterName=FoamCore, defaultValue=<rgba 0.5542 0.7776 0.8 1>}.RGB
   Alpha=Saturate(None=Power(Base=@n24.R, Exp=@n25))
  )
  n27 = ParticleColor()
  n28 = ComponentMask(None=TextureCoordinate()) {r=true}
  n29 = Constant() {r=0.5}
  n30 = TextureCoordinate()
  n31 = ScalarParameter() {defaultValue=1, parameterName=GradTiling}
  n32 = Multiply(
   A=SphereMask(A=Power(Base=@n28) {constExponent=0.6}, B=@n29, Radius=@n29) {hardnessPercent=70}
   B=TextureSample(UVs=Multiply(A=@n30, B=@n31)) {samplerType=SAMPLERTYPE_LinearGrayscale, texture=/Game/Gadgets/Assets/VinderTech_Umbrella/Umbrella_Water/New_Graph_Grad2}.G
  )
  n33 = ParticleColor()
  n34 = Power(Base=@n24.R, Exp=Divide(A=@n25))
  n35 = Multiply(A=Add(A=Multiply(A=@n33.A, B=@n32), B=Subtract(A=@n34)), B=ScalarParameter() {defaultValue=2, parameterName=Opacity})
  n36 = VectorParameter() {parameterName=waterTiling, defaultValue=<rgba 1 2 0 1>}
  n37 = Multiply(A=@n9, B=Multiply(A=@n5.Tile, B=AppendVector(A=@n36.R, B=@n36.R))) {constB=2}
  n38 = Add(
   A=Subtract(A=TextureSample(UVs=@n37) {samplerType=SAMPLERTYPE_LinearColor, texture=/Game/Global/Textures/Default/Caustics/03_WaterFoam_AccentSwirls_H}.R)
   B=Multiply(A=@n32, B=ScalarParameter() {defaultValue=2, parameterName=AccentAmount})
  )
  n39 = ScalarParameter() {defaultValue=0.5, parameterName=WaterBlendPower}
  n40 = Saturate(None=Multiply(A=Saturate(None=@n38), B=Power(Base=@n32, Exp=@n39) {constExponent=3}))
  n41 = Saturate(None=Multiply(A=@n32, B=Add(A=@n35, B=@n40)))
  n42 = ShadingPathSwitch(Default=DepthFade(Opacity=@n41, FadeDistance=ScalarParameter() {defaultValue=8, parameterName=Fade Distance}), Mobile=@n41) {desc=Migrated from FeatureLevelSwitch}
  n43 = DynamicParameter() {paramNames [WPO,PanAdd,Tile,Param4]}
  n44 = WorldPosition()
  n45 = MaterialFunctionCall() {materialFunction=&ObjectPivotPoint}
  n46 = VectorParameter() {parameterName=EndOffset, defaultValue=<rgba 0 0 -125 1>}
  n47 = TextureCoordinate()
  n48 = LinearInterpolate(
   A=Add(A=Subtract(A=@n44.XYZ, B=@n45.Object Pivot Location), B=Multiply(A=@n43.WPO, B=@n46.RGB))
   Alpha=Power(Base=ComponentMask(None=@n47) {r=true}) {constExponent=0.6}
  )
  MaterialFunctionCall(In=SetMaterialAttributes!(
    MaterialAttributes
    Emissive Color=Add(A=@n19, B=@n26)
    Opacity=Multiply(A=@n27.A, B=@n42)
    World Position Offset=Multiply(A=@n43.WPO, B=@n48)
   ) {attributeSetTypes [#1, #2, #3]}) {materialFunction=&MF_TechArtOverrides_accessories}
 ]
}
