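# Base masked subsurface water material for the VinderTech Umbrella gadget, usable on mesh particles and driven by WaterAmount, WaveIntensity and TimeAdd dynamic parameters.
# Rotating Perlin noise over normalized pre-skinned position, weighted by a gradient texture, drives wave World Position Offset.
# Panning water normal and height textures feed Fresnel core-to-edge color and emissive foam, with added color on mobile.
# Tiled ring gradient, height and WaterAmount form a dithered dissolve mask.
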Material /BRCosmetics/Gadgets/Assets/VinderTech_Umbrella/Umbrella_Water/M_Umbrella_Water_Base
graph {
 properties {
  bUseMaterialAttributes=true
  bUsedWithMeshParticles=true
  bUsedWithNiagaraMeshParticles=true
  blendMode=BLEND_Masked
  ditherOpacityMask=true
  pixelDepthOffsetMode=PDOM_Legacy
  shadingModel=MSM_Subsurface
  usedShadingModels=Subsurface
 }
 nodes [
  n1 = Divide(A=PreSkinnedPosition(), B=PreSkinnedLocalBounds().Extents)
  n2 = ScalarParameter() {defaultValue=0.2, parameterName=X}
  n3 = ScalarParameter() {defaultValue=0.2, parameterName=Y}
  n4 = Multiply(
   A=Add(A=DynamicParameter() {paramNames [WaterAmount, WaveIntensity, TimeAdd, Param4]}.TimeAdd, B=Time())
   B=ScalarParameter() {defaultValue=-1.33, parameterName=TimeScale}
  )
  n5 = Rotator(Coordinate=Multiply(A=ComponentMask(None=@n1) {g=true, r=true}, B=AppendVector(A=@n2, B=@n3)), Time=@n4) {centerX=0.1, centerY=0, speed=-2}
  n6 = TextureSample(UVs=@n5) {texture=/Game/Global/Textures/Default/Noise/3xPerlin_M}
  n7 = ConstantBiasScale(None=@n6.R) {bias=-0.5, scale=1}
  n8 = DynamicParameter() {paramNames [WaterAmount, WaveIntensity, TimeAdd, Param4]}
  n9 = ScalarParameter() {defaultValue=5, parameterName=WaveIntensity}
  n10 = TextureObject() {samplerType=SAMPLERTYPE_Masks, texture=&/New_Graph_Grad}
  n11 = TextureSample(UVs=TextureCoordinate(), Tex=@n10) {samplerType=SAMPLERTYPE_Masks, texture=&New_Graph_Grad}
  n12 = Multiply(A=Multiply(A=@n8.WaveIntensity, B=Multiply(A=@n7, B=@n9)), B=@n11.G)
  Add(A=Multiply(A=@n7, B=VertexNormalWS()), B=@n12)
  Constant()
  n13 = VectorParameter() {parameterName=NormalTiling, defaultValue=<rgba 1 1.8 0 1>}
  n14 = TextureCoordinate() {uTiling=0.5, vTiling=2}
  n15 = ScalarParameter() {defaultValue=2, parameterName=Global Tiling}
  n16 = ComponentMask(None=Multiply(A=AppendVector(A=@n13.R, B=@n13.R), B=Multiply(A=@n14, B=@n15))) {g=true, r=true}
  n17 = ScalarParameter() {defaultValue=-0.2, parameterName=NormalX Scale}
  n18 = ScalarParameter() {defaultValue=0.55, parameterName=NormalY Scale}
  n19 = Add(A=@n16, B=AppendVector(A=Multiply(A=@n17, B=@n4), B=Multiply(A=@n18, B=@n4)))
  n20 = TextureSample!(UVs=@n19, Tex, Bias=Constant() {r=2}, Apply View MipBias) {mipValueMode=TMVM_MipBias, samplerType=SAMPLERTYPE_Normal, texture=/Game/Athena/Apollo/Water/Textures/02_Rounded/02_WaterFoam_Veroni_base_N}
  n21 = Power(Base=@n11.G)
  n22 = MaterialFunctionCall(
   Normal=@n20.RGB
   Flatness=LinearInterpolate(
    A=ScalarParameter() {parameterName=DetailNormalFlatness}
    B=ScalarParameter() {defaultValue=0.5, parameterName=DetailNormalFlatness2}
    Alpha=Saturate(None=@n21)
   ) {constB=0.8}
  ) {materialFunction=&FlattenNormal}
  n23 = Fresnel(ExponentIn=ScalarParameter() {defaultValue=2, parameterName=exp}, Normal=Transform(None=@n22.Result)) {baseReflectFraction=0, exponent=0.25}
  n24 = LinearInterpolate(
   A=VectorParameter() {parameterName=Edge Color, defaultValue=<rgba 0 0.1497 0.3281 1>}.RGB
   B=VectorParameter() {parameterName=Core Color, defaultValue=<rgba 0.1489 0.4948 0.3899 1>}.RGB
   Alpha=Saturate(None=Multiply(A=@n23))
  )
  n25 = TextureSample!(UVs=@n19, Tex, Apply View MipBias) {samplerType=SAMPLERTYPE_LinearColor, texture=/Game/Athena/Apollo/Water/Textures/02_Rounded/02_WaterFoam_Veroni_base_H}
  n26 = Power(Base=@n25.R)
  n27 = ConstantBiasScale(None=@n26) {scale=0.2}
  n28 = Constant() {r=1}
  n29 = VertexInterpolator(VS=TextureSample(UVs=@n5, Tex=@n10) {samplerType=SAMPLERTYPE_Masks, texture=&New_Graph_Grad}.R)
  n30 = ScalarParameter() {defaultValue=0.8, parameterName=SubsurfaceRatio}
  n31 = Multiply(A=Multiply(A=@n24, B=LinearInterpolate(A=@n27, B=@n28, Alpha=@n29.PS)), B=Subtract(B=@n30))
  n32 = LinearInterpolate(Alpha=@n11.G) {constA=0.7}
  n33 = Subtract(A=Saturate(None=@n26))
  n34 = ScalarParameter() {defaultValue=0.75, parameterName=asdfasdfa}
  n35 = ScalarParameter() {defaultValue=0.2, parameterName=FoamRadius}
  n36 = LinearInterpolate(A=@n35, B=Multiply(A=@n35) {constB=1.2}, Alpha=@n11.R)
  n37 = ScalarParameter() {defaultValue=0.3, parameterName=Hardness}
  n38 = Multiply(A=LinearInterpolate(Alpha=@n29.PS) {constA=0.25}, B=Clamp(None=SphereMask(A=@n11.G, B=@n34, Radius=@n36, Hardness=@n37)))
  n39 = LinearInterpolate(B=VectorParameter() {parameterName=Foam Color, defaultValue=<rgba 0.5542 0.7776 0.8 1>}.RGB, Alpha=Saturate(None=Add(A=@n33, B=@n38)))
  n40 = Multiply(A=@n24, B=@n30)
  n41 = Add(A=@n39, B=@n40)
  n42 = TextureCoordinate()
  n43 = ScalarParameter() {defaultValue=5, parameterName=Rings}
  n44 = Multiply(
   A=TextureSample(UVs=Multiply(A=@n42, B=@n43)) {samplerType=SAMPLERTYPE_LinearGrayscale, texture=/BRCosmetics/Weapons/WeaponSkins/Wraps/Textures/GlowBro1/T_Wrap_GlowBro1_Grad}.RGB
   B=@n25.R
  )
  n45 = Add(A=Subtract(A=Clamp(None=@n44) {minDefault=0.1}), B=Subtract(A=@n8.WaterAmount) {constB=0.2})
  n46 = SetMaterialAttributes!(
   MaterialAttributes
   Base Color=NamedRerouteDeclaration(None=Multiply(A=@n31, B=@n32)) {name=basecolor, nodeColor=<rgba 0 0.625 1 1>}
   Metallic=NamedRerouteDeclaration(None=ScalarParameter() {defaultValue=0.1, parameterName=Metallic}) {name=metallic, nodeColor=<rgba 0 0.2734 1 1>}
   Specular=NamedRerouteDeclaration(None=ScalarParameter() {defaultValue=1, parameterName=Spec}) {name=specular, nodeColor=<rgba 0.1719 1 0 1>}
   Roughness=NamedRerouteDeclaration(None=ScalarParameter() {defaultValue=0.15, parameterName=Roughness}) {name=roughness, nodeColor=<rgba 1 0 0.4922 1>}
   Emissive Color=NamedRerouteDeclaration(None=ShadingPathSwitch(Default=@n39, Mobile=@n41)) {name=emissive, nodeColor=<rgba 1 0 0.8203 1>}
   Opacity=NamedRerouteDeclaration(None=ScalarParameter() {defaultValue=1, parameterName=SSS}) {name=OPACTIY, nodeColor=<rgba 0.0078 0 1 1>}
   Opacity Mask=NamedRerouteDeclaration(None=MaterialFunctionCall!(In=@n45, Ramp Width) {materialFunction=/Engine/Functions/Engine_MaterialFunctions01/Math/SmoothCeil}.Result) {name=opactiymask, nodeColor=<rgba 0.0078 0 1 1>}
   Normal=NamedRerouteDeclaration(None=@n22.Result) {name=normal, nodeColor=<rgba 1 0 0.0469 1>}
   Subsurface Color=NamedRerouteDeclaration(None=@n40) {name=sss, nodeColor=<rgba 0 1 0.8125 1>}
   World Position Offset=NamedRerouteDeclaration(None=@n12) {name=wpo, nodeColor=<rgba 0 1 0.0625 1>}
  ) {attributeSetTypes [
    #1
    #2
    #3
    #4
    #5
    #6
    #7
    #8
    #9
    #10
   ]}
  MaterialFunctionCall(In=@n46) {materialFunction=&MF_TechArtOverrides_accessories}
  ScalarParameter() {defaultValue=1, parameterName=WaterAmount}
  ShadingModel()
  VertexInterpolator(VS=@n6.R)
  VertexInterpolator(VS=TextureSample(UVs=@n16, Tex=@n10) {samplerType=SAMPLERTYPE_Masks, texture=&New_Graph_Grad}.B)
 ]
}
