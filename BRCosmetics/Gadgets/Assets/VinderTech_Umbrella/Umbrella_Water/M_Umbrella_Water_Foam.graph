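# Translucent subsurface foam water material for the VinderTech umbrella gadget, built for Niagara and mesh particles.
# Time-panned foam normal and height textures drive Fresnel-tinted color, roughness and emissive foam.
# Gradient, foam and swirl masks form depth-faded opacity, cut by a ring gradient against particle WaterAmount through SmoothCeil.
# Dynamic parameters also offset time and scale rotating-noise vertex waves, which the deferred shading path replaces with a constant.
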Material /BRCosmetics/Gadgets/Assets/VinderTech_Umbrella/Umbrella_Water/M_Umbrella_Water_Foam
graph {
 properties {
  bUseMaterialAttributes=true
  bUsedWithMeshParticles=true
  bUsedWithNiagaraMeshParticles=true
  blendMode=BLEND_Translucent
  ditherOpacityMask=true
  pixelDepthOffsetMode=PDOM_Legacy
  shadingModel=MSM_Subsurface
  translucencyLightingMode=TLM_SurfacePerPixelLighting
  usedShadingModels=Subsurface
 }
 nodes [
  n1 = TextureCoordinate()
  n2 = ScalarParameter() {defaultValue=1, parameterName=FoamTiling}
  n3 = VectorParameter() {parameterName=NormalTiling, defaultValue=<rgba 2 1 0 1>}
  n4 = ScalarParameter() {defaultValue=-0.4, parameterName=NormalX Scale}
  n5 = Multiply(
   A=Add(A=DynamicParameter() {paramNames [WaterAmount, WaveIntensity, TimeAdd, Param4]}.TimeAdd, B=Time())
   B=ScalarParameter() {defaultValue=-1.33, parameterName=TimeScale}
  )
  n6 = ScalarParameter() {defaultValue=0.65, parameterName=NormalY Scale}
  n7 = Add(A=Multiply(A=Multiply(A=@n1, B=@n2), B=AppendVector(A=@n3.R, B=@n3.R)), B=AppendVector(A=Multiply(A=@n4, B=@n5), B=Multiply(A=@n6, B=@n5)))
  n8 = TextureSample!(UVs=@n7, Tex, Bias=Constant() {r=2}, Apply View MipBias) {
   mipValueMode=TMVM_MipBias
   samplerSource=SSM_Wrap_WorldGroupSettings
   samplerType=SAMPLERTYPE_Normal
   texture=/Game/Athena/Apollo/Water/Textures/02_Rounded/02_WaterFoam_Veroni_base_N
  }
  n9 = MaterialFunctionCall(Normal=@n8.RGB, Flatness=ScalarParameter() {defaultValue=0.5, parameterName=Flattne}) {materialFunction=&FlattenNormal}
  n10 = Multiply(A=Fresnel(ExponentIn=ScalarParameter() {defaultValue=2, parameterName=exp}, Normal=Transform(None=@n9.Result)) {baseReflectFraction=0, exponent=0.25})
  n11 = Multiply(
   A=LinearInterpolate(
    A=VectorParameter() {parameterName=Edge Color, defaultValue=<rgba 0 0.1497 0.3281 1>}.RGB
    B=VectorParameter() {parameterName=Core Color, defaultValue=<rgba 0.1489 0.4948 0.3899 1>}.RGB
    Alpha=Saturate(None=@n10)
   )
   B=ScalarParameter() {defaultValue=0.5, parameterName=Brightness}
  )
  n12 = ScalarParameter() {defaultValue=0.8, parameterName=SubsurfaceRatio}
  n13 = TextureSample!(UVs=@n7, Tex, Apply View MipBias) {
   samplerSource=SSM_Wrap_WorldGroupSettings
   samplerType=SAMPLERTYPE_LinearColor
   texture=/Game/Athena/Apollo/Water/Textures/02_Rounded/02_WaterFoam_Veroni_base_H
  }
  n14 = ScalarParameter() {defaultValue=4, parameterName=FoamAmount}
  n15 = Saturate(None=Power(Base=@n13.R, Exp=@n14))
  n16 = TextureSample(UVs=Multiply(A=TextureCoordinate(), B=ScalarParameter() {defaultValue=1, parameterName=GradTiling})) {
   samplerSource=SSM_Wrap_WorldGroupSettings
   samplerType=SAMPLERTYPE_LinearGrayscale
   texture=/Game/Gadgets/Assets/VinderTech_Umbrella/Umbrella_Water/New_Graph_Grad2
  }
  n17 = Power(Base=@n13.R, Exp=Divide(A=@n14))
  n18 = Multiply(A=Add(A=@n16.R, B=Subtract(A=@n17)), B=ScalarParameter() {defaultValue=2, parameterName=Opacity})
  n19 = VectorParameter() {parameterName=waterTiling, defaultValue=<rgba 2 2 0 1>}
  n20 = TextureSample(UVs=Multiply(A=@n7, B=AppendVector(A=@n19.R, B=@n19.R)) {constB=2}) {
   samplerSource=SSM_Wrap_WorldGroupSettings
   samplerType=SAMPLERTYPE_LinearColor
   texture=/Game/Global/Textures/Default/Caustics/03_WaterFoam_AccentSwirls_H
  }
  n21 = Add(A=Subtract(A=@n20.G), B=Multiply(A=@n16.R, B=ScalarParameter() {defaultValue=2, parameterName=AccentAmount}))
  n22 = ScalarParameter() {defaultValue=0.5, parameterName=WaterBlendPower}
  n23 = Saturate(None=Multiply(A=Saturate(None=@n21), B=Power(Base=@n16.R, Exp=@n22) {constExponent=3}))
  n24 = Saturate(None=Multiply(A=@n16.R, B=Add(A=@n18, B=@n23)))
  n25 = ScalarParameter() {defaultValue=2, parameterName=Fade Distance}
  n26 = TextureSample(UVs=Multiply(A=TextureCoordinate(), B=ScalarParameter() {defaultValue=5, parameterName=Rings})) {
   samplerSource=SSM_Wrap_WorldGroupSettings
   samplerType=SAMPLERTYPE_LinearGrayscale
   texture=/BRCosmetics/Weapons/WeaponSkins/Wraps/Textures/GlowBro1/T_Wrap_GlowBro1_Grad
  }
  n27 = Multiply(A=@n26.RGB)
  n28 = ScalarParameter() {defaultValue=1, parameterName=WaterAmount}
  n29 = DynamicParameter() {paramNames [WaterAmount, WaveIntensity, TimeAdd, Param4]}
  n30 = Add(A=Subtract(A=Clamp(None=@n27) {minDefault=0.1}), B=Subtract(A=Multiply(A=@n28, B=@n29.WaterAmount)) {constB=0.2})
  n31 = ComponentMask(None=Divide(A=PreSkinnedPosition(), B=PreSkinnedLocalBounds().Extents)) {g=true, r=true}
  n32 = AppendVector(A=ScalarParameter() {defaultValue=0.2, parameterName=X}, B=ScalarParameter() {defaultValue=0.2, parameterName=Y})
  n33 = TextureSample(UVs=Rotator(Coordinate=Multiply(A=@n31, B=@n32), Time=@n5) {centerX=0.1, centerY=0, speed=-2}) {texture=/Game/Global/Textures/Default/Noise/3xPerlin_M}
  n34 = Multiply(A=ConstantBiasScale(None=@n33.R) {bias=-0.5, scale=1}, B=ScalarParameter() {defaultValue=5, parameterName=WaveIntensity})
  n35 = SetMaterialAttributes!(
   MaterialAttributes
   Base Color=Multiply(A=@n11, B=Subtract(B=@n12))
   Metallic=Constant() {r=0.2}
   Specular=Constant() {r=0.2}
   Roughness=LinearInterpolate(A=Constant() {r=0.1}, B=Constant() {r=0.75}, Alpha=@n15)
   Emissive Color=LinearInterpolate(A=@n11, B=VectorParameter() {parameterName=FoamCore, defaultValue=<rgba 0.5542 0.7776 0.8 1>}.RGB, Alpha=@n15)
   Opacity=Multiply(
    A=DepthFade(Opacity=@n24, FadeDistance=@n25)
    B=MaterialFunctionCall!(In=@n30, Ramp Width) {materialFunction=/Engine/Functions/Engine_MaterialFunctions01/Math/SmoothCeil}.Result
   )
   Normal=@n9.Result
   World Position Offset=ShadingPathSwitch(Default=Multiply(A=@n29.WaveIntensity, B=@n34), Deferred=Constant()) {desc=Migrated from FeatureLevelSwitch}
   Subsurface Color=Multiply(A=@n11, B=@n12)
  ) {attributeSetTypes [
    #1
    #2
    #3
    #4
    #5
    #6
    #7
    #8
    #9
   ]}
  MaterialFunctionCall(In=@n35) {materialFunction=&MF_TechArtOverrides_accessories}
  VertexInterpolator(VS=@n33.R)
 ]
}
